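# Conversation database for the IcicleLarva NPC, who introduces himself as Superman, defining his greeting and the hub of his dialogue.
# The entry point leads to one Speech node that has a default line plus contextual lines gated by requirements: first-ever conversation, per-NPC match conversation count, and a specific worn outfit.
# Priority curves rank the first-meeting and outfit reactions.
# The speech then branches through link nodes to NPC-specific dialogue and quests, basic services, and shared NPC-library quests.
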ConversationDatabase /NPCLibrary/NPCs/IcicleLarva/NPCLibrary_Conversation_IcicleLarva
entry_points [{tag=Conversation.Entry.IcicleLarva, destinations [#1]}]
nodes [
 ConversationEntryPointNode_7 {
  id=#1
  class=ConversationEntryPointNode
  outputs [#2]
  properties {entryTag {tagName=Conversation.Entry.IcicleLarva}}
 }
 Speech_C_4 {
  id=#2
  class=Speech_C
  outputs [#3, #4, #5, #6]
  properties {
   nodeName=""
   generalConfig {
    defaultMessage="Innocent people need protecting. "
    contextualMessages [
     {
      message=Pleased to meet you. They call me Superman.
      contextRequirements [{requirement {class=FortControllerRequirement_IsFirstEverConversationWithNPC}}]
      priority {curve {rowName=Tandem.Speech.Priority.FirstEverInteraction}}
     }
     {
      message="Defending truth, tolerance, and justice. That's what I'm about."
      contextRequirements [{requirement {class=FortControllerRequirement_CheckNPCMatchConversationCount, properties {nPCUniqueIdTag {tagName=AISpawnerData.Type.Tandem.IcicleLarva}}}}]
      priority {curve {rowName=None}}
     }
     {
      message="Innocent people need protecting. "
      contextRequirements [{requirement {class=FortControllerRequirement_CheckNPCMatchConversationCount, properties {nPCUniqueIdTag {tagName=AISpawnerData.Type.Tandem.IcicleLarva}}}}]
      priority {curve {rowName=None}}
     }
     {
      message="The least we can do, in this world, is be kind to one another."
      contextRequirements [{requirement {class=FortControllerRequirement_CheckNPCMatchConversationCount, properties {nPCUniqueIdTag {tagName=AISpawnerData.Type.Tandem.IcicleLarva}}}}]
      priority {curve {rowName=None}}
     }
     {
      message=Have we... met...?
      contextRequirements [{requirement {
         class=FortControllerRequirement_HasCID
         properties {allowedCIds [/BRCosmetics/Athena/Items/Cosmetics/Characters/CID_A_109_Athena_Commando_M_Emperor]}
        }}]
      priority {curve {rowName=Tandem.Speech.Priority.OutfitReaction}}
     }
    ]
   }
  }
 }
 ConversationLinkNode_3 {
  id=#6
  class=ConversationLinkNode
  properties {remoteEntryTag {tagName=Conversation.Entry.IcicleLarva.Dialogue}}
 }
 ConversationLinkNode_2 {id=#5, class=ConversationLinkNode, properties {remoteEntryTag {tagName=Conversation.Entry.IcicleLarva.Quests}}}
 ConversationLinkNode_0 {id=#4, class=ConversationLinkNode, properties {remoteEntryTag {tagName=Conversation.Entry.BasicServices}}}
 ConversationLinkNode_1 {id=#3, class=ConversationLinkNode, properties {remoteEntryTag {tagName=Conversation.Entry.NPCLibrary.Quests}}}
]
